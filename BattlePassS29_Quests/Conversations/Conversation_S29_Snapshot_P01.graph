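# Conversation database for the Season 29 Snapshot Part 01 battle pass quests, holding mythology-themed dialogue for Q01, Q04, Q06, Q10 and Q14.
# Each tagged entry point starts a Speech chain ending on a quest-specific choice tag, likely for objective tracking.
# Q06 (Hades) and Q10 (Ares) require their quest to be active and issue combat challenges.
# Q04 is an Oracle food quiz where wrong answers route to a Back node.
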ConversationDatabase /BattlePassS29_Quests/Conversations/Conversation_S29_Snapshot_P01
entry_points [
 {tag=Conversation.Entry.Season29.Snapshot.P01.Q01, destinations [#1]}
 {tag=Conversation.Entry.Season29.Snapshot.P01.Q06, destinations [#2]}
 {tag=Conversation.Entry.Season29.Snapshot.P01.Q10, destinations [#3]}
 {tag=Conversation.Entry.Season29.Snapshot.P01.Q14, destinations [#4]}
 {tag=Conversation.Entry.Season29.Snapshot.P01.Q04, destinations [#5]}
]
nodes [
 ConversationEntryPointNode_4 {
  id=#5
  class=ConversationEntryPointNode
  outputs [#6]
  properties {entryTag {tagName=Conversation.Entry.Season29.Snapshot.P01.Q04}}
 }
 Speech_C_13 {
  id=#6
  class=Speech_C
  outputs [#7]
  properties {generalConfig {defaultMessage="A final test awaits, mortal. Are you prepared? ", defaultMetadataParameters ["" {value=""}]}}
  sub_nodes [ConversationChoiceNode_18 {
    class=ConversationChoiceNode
    properties {defaultChoiceDisplayText=Continue, choiceTags {gameplayTags [{tagName=Conversation.Choice.Category.MainQuestline}]}}
   }]
 }
 Speech_C_11 {
  id=#7
  class=Speech_C
  outputs [#8, #9, #10]
  properties {generalConfig {defaultMessage=Which food is most appealing to the gods?, defaultMetadataParameters [OracleEmote {value=Question}]}}
  sub_nodes [ConversationChoiceNode_17 {class=ConversationChoiceNode, properties {defaultChoiceDisplayText=Continue}}]
 }
 Speech_C_8 {
  id=#10
  class=Speech_C
  outputs [#11]
  properties {generalConfig {defaultMessage="You have chosen... unwisely. ", defaultMetadataParameters [OracleEmote {value=Awkward}]}}
  sub_nodes [ConversationChoiceNode_2 {class=ConversationChoiceNode, properties {defaultChoiceDisplayText=Durrr Burger}}]
 }
 Back_C_22 {id=#11, class=Back_C}
 Speech_C_7 {
  id=#9
  class=Speech_C
  outputs []
  properties {generalConfig {defaultMessage=Good work., defaultMetadataParameters [OracleEmote {value=Exclamation}]}}
  sub_nodes [ConversationChoiceNode_9 {
    class=ConversationChoiceNode
    properties {defaultChoiceDisplayText=Banana, choiceTags {gameplayTags [{tagName=Athena.Quests.S29.Conversations.Snapshot.P01.Choice.Q04}]}}
   }]
 }
 Speech_C_6 {
  id=#8
  class=Speech_C
  outputs [#11]
  properties {generalConfig {defaultMessage="You have chosen... unwisely. ", defaultMetadataParameters [OracleEmote {value=Awkward}]}}
  sub_nodes [ConversationChoiceNode_8 {class=ConversationChoiceNode, properties {defaultChoiceDisplayText=Pizza}}]
 }
 ConversationEntryPointNode_3 {
  id=#4
  class=ConversationEntryPointNode
  outputs [#12]
  properties {entryTag {tagName=Conversation.Entry.Season29.Snapshot.P01.Q14}}
 }
 Speech_C_15 {
  id=#12
  class=Speech_C
  outputs []
  properties {generalConfig {defaultMessage=Zeus's closest allies have seen you fight..., defaultMetadataParameters [OracleEmote {value=Exclamation}]}}
  sub_nodes [ConversationChoiceNode_16 {
    class=ConversationChoiceNode
    properties {
     defaultChoiceDisplayText=Continue
     choiceTags {gameplayTags [
       {tagName=Athena.Quests.S29.Conversations.Snapshot.P01.Choice.Q14}
       {tagName=Conversation.Choice.Category.MainQuestline}
       {tagName=Conversation.Choice.AutoSelectIfValid}
      ]}
    }
   }]
 }
 ConversationEntryPointNode_2 {
  id=#3
  class=ConversationEntryPointNode
  outputs [#13]
  properties {entryTag {tagName=Conversation.Entry.Season29.Snapshot.P01.Q10}}
 }
 Speech_C_10 {
  id=#13
  class=Speech_C
  outputs [#14]
  properties {generalConfig {defaultMessage=WHO DARES CHALLENGE ZEUS'S MIGHTY PANTHEON?, defaultMetadataParameters [Ares {value=""}]}}
  sub_nodes [
   ConversationChoiceNode_15 {
    class=ConversationChoiceNode
    properties {
     defaultChoiceDisplayText=Continue
     choiceTags {gameplayTags [{tagName=Conversation.Choice.Category.MainQuestline}, {tagName=Conversation.Choice.AutoSelectIfValid}]}
    }
   }
   FortConversationRequirement_ControllerRequirement_3 {
    class=FortConversationRequirement_ControllerRequirement
    properties {
     nodeName=Require: Is Quest Active
     requirement {
      class=FortControllerRequirement_IsQuestActive
      properties {questDefinition=/BattlePassS29_Quests/Items/QuestItems/Snapshots/Part01/quest_S29_snapshot_p01_q10}
     }
    }
   }
  ]
 }
 Speech_C_0 {
  id=#14
  class=Speech_C
  outputs [#15]
  properties {generalConfig {defaultMessage="HA! YOU WOULDN'T KNOW AN HONORABLE BATTLE IF IT SLEW YOU WHERE YOU STOOD!"}}
  sub_nodes [ConversationChoiceNode_14 {class=ConversationChoiceNode, properties {defaultChoiceDisplayText=Continue}}]
 }
 Speech_C_16 {
  id=#15
  class=Speech_C
  outputs [#16]
  properties {generalConfig {defaultMessage=PROVE YOUR HONOR. ARM YOURSELF WITH MY FAVORED WEAPON... AND PREPARE YOURSELF FOR COMBAT.}}
  sub_nodes [ConversationChoiceNode_19 {class=ConversationChoiceNode, properties {defaultChoiceDisplayText=Continue}}]
 }
 Speech_C_17 {
  id=#16
  class=Speech_C
  outputs [#17]
  properties {generalConfig {defaultMessage="IF YOU CAN COLLECT A MEDALLION, I WILL CONSIDER YOU WORTHY."}}
  sub_nodes [ConversationChoiceNode_3 {class=ConversationChoiceNode, properties {defaultChoiceDisplayText=Continue}}]
 }
 Speech_C_4 {
  id=#17
  class=Speech_C
  outputs []
  properties {generalConfig {defaultMessage="IT'LL NEVER HAPPEN OF COURSE. HA!"}}
  sub_nodes [ConversationChoiceNode_12 {
    class=ConversationChoiceNode
    properties {
     defaultChoiceDisplayText=Continue
     choiceTags {gameplayTags [
       {tagName=Conversation.Choice.Category.MainQuestline}
       {tagName=Conversation.Choice.AutoSelectIfValid}
       {tagName=Athena.Quests.S29.Conversations.Snapshot.P01.Choice.Q10}
      ]}
    }
   }]
 }
 ConversationEntryPointNode_0 {
  id=#2
  class=ConversationEntryPointNode
  outputs [#18]
  properties {entryTag {tagName=Conversation.Entry.Season29.Snapshot.P01.Q06}}
 }
 Speech_C_3 {
  id=#18
  class=Speech_C
  outputs [#19]
  properties {generalConfig {defaultMessage="YOU'RE mortalkind's hero?! ...Good luck with that.", defaultMetadataParameters [Hades {value=""}]}}
  sub_nodes [
   ConversationChoiceNode_10 {
    class=ConversationChoiceNode
    properties {
     defaultChoiceDisplayText=Continue
     choiceTags {gameplayTags [{tagName=Conversation.Choice.Category.MainQuestline}, {tagName=Conversation.Choice.AutoSelectIfValid}]}
    }
   }
   FortConversationRequirement_ControllerRequirement_2 {
    class=FortConversationRequirement_ControllerRequirement
    properties {
     nodeName=Require: Is Quest Active
     requirement {
      class=FortControllerRequirement_IsQuestActive
      properties {questDefinition=/BattlePassS29_Quests/Items/QuestItems/Snapshots/Part01/quest_S29_snapshot_p01_q06}
     }
    }
   }
  ]
 }
 Speech_C_9 {
  id=#19
  class=Speech_C
  outputs [#20]
  properties {generalConfig {defaultMessage="If you're so sure you can defeat my brother, prove it."}}
  sub_nodes [ConversationChoiceNode_0 {
    class=ConversationChoiceNode
    properties {defaultChoiceDisplayText=Continue, choiceTags {gameplayTags [{tagName=Conversation.Choice.Category.MainQuestline}]}}
   }]
 }
 Speech_C_12 {
  id=#20
  class=Speech_C
  outputs [#21]
  properties {generalConfig {defaultMessage="Win a fight against my beloved dog, Cerberus. "}}
  sub_nodes [ConversationChoiceNode_11 {
    class=ConversationChoiceNode
    properties {defaultChoiceDisplayText=Continue, choiceTags {gameplayTags [{tagName=Conversation.Choice.Category.MainQuestline}]}}
   }]
 }
 Speech_C_14 {
  id=#21
  class=Speech_C
  outputs []
  properties {generalConfig {defaultMessage=I'll even let you use a powerful weapon to do it. Seems like you'd need the help.}}
  sub_nodes [ConversationChoiceNode_5 {
    class=ConversationChoiceNode
    properties {
     defaultChoiceDisplayText=Continue
     choiceTags {gameplayTags [{tagName=Conversation.Choice.Category.MainQuestline}, {tagName=Athena.Quests.S29.Conversations.Snapshot.P01.Choice.Q06}]}
    }
   }]
 }
 ConversationEntryPointNode_1 {
  id=#1
  class=ConversationEntryPointNode
  outputs [#22]
  properties {entryTag {tagName=Conversation.Entry.Season29.Snapshot.P01.Q01}}
 }
 Speech_C_1 {
  id=#22
  class=Speech_C
  outputs []
  properties {generalConfig {defaultMessage=Zeus' wrath is upon you..., defaultMetadataParameters [OracleEmote {value=Idle}]}}
  sub_nodes [ConversationChoiceNode_13 {
    class=ConversationChoiceNode
    properties {
     defaultChoiceDisplayText=Continue
     choiceTags {gameplayTags [
       {tagName=Athena.Quests.S29.Conversations.Snapshot.P01.Choice.Q01}
       {tagName=Conversation.Choice.Category.MainQuestline}
       {tagName=Conversation.Choice.AutoSelectIfValid}
      ]}
    }
   }]
 }
]
