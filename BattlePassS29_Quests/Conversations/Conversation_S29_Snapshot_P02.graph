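# Conversation database for the Season 29 Snapshot Part 02 quests Q01, Q04 and Q12, holding short NPC dialogues about opposing Zeus and cheering up Aphrodite.
# Each entry tag starts a linear Speech chain with Continue choices, auto-selected into the main questline category at the first line.
# Q04 and Q12 openers require an active quest with an unfinished objective, while Q01 is ungated.
# Each final choice carries a quest-specific gameplay tag.
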ConversationDatabase /BattlePassS29_Quests/Conversations/Conversation_S29_Snapshot_P02
entry_points [
 {tag=Conversation.Entry.Season29.Snapshot.P02.Q01, destinations [#1]}
 {tag=Conversation.Entry.Season29.Snapshot.P02.Q04, destinations [#2]}
 {tag=Conversation.Entry.Season29.Snapshot.P02.Q12, destinations [#3]}
]
nodes [
 ConversationEntryPointNode_6 {
  id=#3
  class=ConversationEntryPointNode
  outputs [#4]
  properties {entryTag {tagName=Conversation.Entry.Season29.Snapshot.P02.Q12}}
 }
 Speech_C_27 {
  id=#4
  class=Speech_C
  outputs [#5]
  properties {generalConfig {defaultMessage="Well met, little mouse. You show some talent for the Eternal Hunt."}}
  sub_nodes [
   ConversationChoiceNode_21 {
    class=ConversationChoiceNode
    properties {
     defaultChoiceDisplayText=Continue
     choiceTags {gameplayTags [{tagName=Conversation.Choice.Category.MainQuestline}, {tagName=Conversation.Choice.AutoSelectIfValid}]}
    }
   }
   FortConversationRequirement_ControllerRequirement_1 {
    class=FortConversationRequirement_ControllerRequirement
    properties {
     nodeName=Require: Is Quest Active
     requirement {
      class=FortControllerRequirement_IsQuestActive
      properties {questDefinition=/BattlePassS29_Quests/Items/QuestItems/Snapshots/Part02/quest_S29_snapshot_p02_q12}
     }
    }
   }
   FortConversationRequirement_ControllerRequirement_3 {
    class=FortConversationRequirement_ControllerRequirement
    properties {
     nodeName=Require: Has Completed Quest Objective
     requirement {
      class=FortControllerRequirement_HasCompletedQuestObjective
      properties {
       bInverseCondition=true
       objectiveBackendName=quest_S29_snapshot_p02_q12_obj1
       questDefinition=&quest_S29_snapshot_p02_q12
      }
     }
    }
   }
  ]
 }
 Speech_C_28 {
  id=#5
  class=Speech_C
  outputs [#6]
  properties {generalConfig {defaultMessage="But that's not what you wish to speak about, is it?"}}
  sub_nodes [ConversationChoiceNode_22 {class=ConversationChoiceNode, properties {defaultChoiceDisplayText=Continue}}]
 }
 Speech_C_29 {
  id=#6
  class=Speech_C
  outputs [#7]
  properties {generalConfig {defaultMessage=Zeus... His plans for mortals violate the rules of the hunt.}}
  sub_nodes [ConversationChoiceNode_31 {class=ConversationChoiceNode, properties {defaultChoiceDisplayText=Continue}}]
 }
 Speech_C_30 {
  id=#7
  class=Speech_C
  outputs []
  properties {generalConfig {defaultMessage="I will stand with you against him. "}}
  sub_nodes [ConversationChoiceNode_20 {
    class=ConversationChoiceNode
    properties {defaultChoiceDisplayText=Continue, choiceTags {gameplayTags [{tagName=Athena.Quests.S29.Conversations.Snapshot.P02.Choice.Q12}]}}
   }]
 }
 ConversationEntryPointNode_5 {
  id=#2
  class=ConversationEntryPointNode
  outputs [#8]
  properties {entryTag {tagName=Conversation.Entry.Season29.Snapshot.P02.Q04}}
 }
 Speech_C_23 {
  id=#8
  class=Speech_C
  outputs [#9]
  properties {generalConfig {defaultMessage="Mortal!! I'm so glad you showed."}}
  sub_nodes [
   ConversationChoiceNode_28 {
    class=ConversationChoiceNode
    properties {
     defaultChoiceDisplayText=Continue
     choiceTags {gameplayTags [{tagName=Conversation.Choice.Category.MainQuestline}, {tagName=Conversation.Choice.AutoSelectIfValid}]}
    }
   }
   FortConversationRequirement_ControllerRequirement_0 {
    class=FortConversationRequirement_ControllerRequirement
    properties {
     nodeName=Require: Is Quest Active
     requirement {
      class=FortControllerRequirement_IsQuestActive
      properties {questDefinition=/BattlePassS29_Quests/Items/QuestItems/Snapshots/Part02/quest_S29_snapshot_p02_q04}
     }
    }
   }
   FortConversationRequirement_ControllerRequirement_2 {
    class=FortConversationRequirement_ControllerRequirement
    properties {
     nodeName=Require: Has Completed Quest Objective
     requirement {
      class=FortControllerRequirement_HasCompletedQuestObjective
      properties {
       bInverseCondition=true
       objectiveBackendName=quest_S29_snapshot_p02_q04_obj0
       questDefinition=&quest_S29_snapshot_p02_q04
      }
     }
    }
   }
  ]
 }
 Speech_C_24 {
  id=#9
  class=Speech_C
  outputs [#10]
  properties {generalConfig {defaultMessage="Aphrodite's been real bummed lately. Her dad's been really, like, 'anti-mortal' lately??"}}
  sub_nodes [ConversationChoiceNode_29 {class=ConversationChoiceNode, properties {defaultChoiceDisplayText=Continue}}]
 }
 Speech_C_26 {
  id=#10
  class=Speech_C
  outputs []
  properties {generalConfig {defaultMessage="I thought we could cheer her up. Or rather... YOU could cheer her up!"}}
  sub_nodes [ConversationChoiceNode_30 {
    class=ConversationChoiceNode
    properties {defaultChoiceDisplayText=Continue, choiceTags {gameplayTags [{tagName=Athena.Quests.S29.Conversations.Snapshot.P02.Choice.Q04}]}}
   }]
 }
 ConversationEntryPointNode_1 {
  id=#1
  class=ConversationEntryPointNode
  outputs [#11]
  properties {entryTag {tagName=Conversation.Entry.Season29.Snapshot.P02.Q01}}
 }
 Speech_C_1 {
  id=#11
  class=Speech_C
  outputs [#12]
  properties {generalConfig {
    defaultMessage="Mortal. You have taken up humanity's cause... But I fear the cause is already lost. "
    defaultMetadataParameters [Aphrodite {value=""}]
   }}
  sub_nodes [ConversationChoiceNode_24 {
    class=ConversationChoiceNode
    properties {
     defaultChoiceDisplayText=Continue
     choiceTags {gameplayTags [{tagName=Conversation.Choice.Category.MainQuestline}, {tagName=Conversation.Choice.AutoSelectIfValid}]}
    }
   }]
 }
 Speech_C_18 {
  id=#12
  class=Speech_C
  outputs [#13]
  properties {generalConfig {defaultMessage="My father... He's a stubborn, imperious immortal... ", defaultMetadataParameters [Aphrodite {value=""}]}}
  sub_nodes [ConversationChoiceNode_25 {class=ConversationChoiceNode, properties {defaultChoiceDisplayText=Continue}}]
 }
 Speech_C_19 {
  id=#13
  class=Speech_C
  outputs [#14]
  properties {generalConfig {defaultMessage=You have the spirit of a true fighter. But your battle has only just begun., defaultMetadataParameters [Aphrodite {value=""}]}}
  sub_nodes [ConversationChoiceNode_26 {class=ConversationChoiceNode, properties {defaultChoiceDisplayText=Continue}}]
 }
 Speech_C_22 {
  id=#14
  class=Speech_C
  outputs []
  properties {generalConfig {
    defaultMessage="Ready yourself. The fight against my father will require your full strength. "
    defaultMetadataParameters [Aphrodite {value=""}]
   }}
  sub_nodes [ConversationChoiceNode_27 {
    class=ConversationChoiceNode
    properties {defaultChoiceDisplayText=Continue, choiceTags {gameplayTags [{tagName=Athena.Quests.S29.Conversations.Snapshot.P02.Choice.Q01}]}}
   }]
 }
]
