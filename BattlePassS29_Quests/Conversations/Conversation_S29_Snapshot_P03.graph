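# Conversation database for Season 29 battle pass Snapshot Part 03 quests, holding NPC dialogue for steps Q01 (Cerberus), Q04 (Hades) and Q08 (an Oracle hint).
# Each entry tag starts a linear speech chain, auto-selected when valid.
# Q01 and Q04 require their quest active with quest and first objective incomplete; Q08 is a single ungated line.
# Choice tags late in each chain identify the quest step, with Q04's on both final lines.
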ConversationDatabase /BattlePassS29_Quests/Conversations/Conversation_S29_Snapshot_P03
entry_points [
 {tag=Conversation.Entry.Season29.Snapshot.P03.Q01, destinations [#1]}
 {tag=Conversation.Entry.Season29.Snapshot.P03.Q04, destinations [#2]}
 {tag=Conversation.Entry.Season29.Snapshot.P03.Q08, destinations [#3]}
]
nodes [
 ConversationEntryPointNode_0 {
  id=#3
  class=ConversationEntryPointNode
  outputs [#4]
  properties {entryTag {tagName=Conversation.Entry.Season29.Snapshot.P03.Q08}}
 }
 Speech_C_2 {
  id=#4
  class=Speech_C
  outputs []
  properties {generalConfig {defaultMessage="Zeus has many enemies, some who would welcome a fight...", defaultMetadataParameters [OracleEmote {value=Idle}]}}
  sub_nodes [ConversationChoiceNode_13 {
    class=ConversationChoiceNode
    properties {
     defaultChoiceDisplayText=Continue
     choiceTags {gameplayTags [
       {tagName=Conversation.Choice.Category.MainQuestline}
       {tagName=Conversation.Choice.AutoSelectIfValid}
       {tagName=Athena.Quests.S29.Conversations.Snapshot.03.08}
      ]}
    }
   }]
 }
 ConversationEntryPointNode_3 {
  id=#2
  class=ConversationEntryPointNode
  outputs [#5]
  properties {entryTag {tagName=Conversation.Entry.Season29.Snapshot.P03.Q04}}
 }
 Speech_C_20 {
  id=#5
  class=Speech_C
  outputs [#6]
  properties {
   softTaskIcon=&T_UI_NPC_MainQuest
   generalConfig {defaultMessage="I will admit, it was... sweet of you to help out Cerberus...", defaultMetadataParameters [Hades {value=""}]}
  }
  sub_nodes [
   ConversationChoiceNode_9 {
    class=ConversationChoiceNode
    properties {
     defaultChoiceDisplayText=Speak with Hades
     choiceTags {gameplayTags [{tagName=Conversation.Choice.Category.MainQuestline}, {tagName=Conversation.Choice.AutoSelectIfValid}]}
    }
   }
   FortConversationRequirement_ControllerRequirement_14 {
    class=FortConversationRequirement_ControllerRequirement
    properties {
     nodeName=Require: Has Completed Quest
     requirement {
      class=FortControllerRequirement_HasCompletedQuest
      properties {bInverseCondition=true, questDefinition=/BattlePassS29_Quests/Items/QuestItems/Snapshots/Part03/quest_S29_snapshot_p03_q04}
     }
    }
   }
   FortConversationRequirement_ControllerRequirement_15 {
    class=FortConversationRequirement_ControllerRequirement
    properties {
     nodeName=Require: Is Quest Active
     requirement {
      class=FortControllerRequirement_IsQuestActive
      properties {questDefinition=&quest_S29_snapshot_p03_q04}
     }
    }
   }
   FortConversationRequirement_ControllerRequirement_16 {
    class=FortConversationRequirement_ControllerRequirement
    properties {
     nodeName=Require: Has Completed Quest Objective
     requirement {
      class=FortControllerRequirement_HasCompletedQuestObjective
      properties {
       bInverseCondition=true
       objectiveBackendName=quest_S29_snapshot_p03_q04_obj0
       questDefinition=&quest_S29_snapshot_p03_q04
      }
     }
    }
   }
  ]
 }
 Speech_C_3 {
  id=#6
  class=Speech_C
  outputs [#7]
  properties {generalConfig {defaultMessage=I took Cerberus'... artifact away for safekeeping during Zeus' tempest.}}
  sub_nodes [ConversationChoiceNode_2 {class=ConversationChoiceNode, properties {defaultChoiceDisplayText=Continue}}]
 }
 Speech_C_4 {
  id=#7
  class=Speech_C
  outputs [#8]
  properties {generalConfig {defaultMessage=I've since realized that taking it away has made him agitated. And desperate to find it himself.}}
  sub_nodes [ConversationChoiceNode_5 {class=ConversationChoiceNode, properties {defaultChoiceDisplayText=Continue}}]
 }
 Speech_C_16 {
  id=#8
  class=Speech_C
  outputs [#9]
  properties {generalConfig {
    defaultMessage="I'm putting the scent of his artifact on you. Pay Cerberus a visit. Play with him. "
    defaultMetadataParameters ["" {value=""}]
   }}
  sub_nodes [ConversationChoiceNode_7 {
    class=ConversationChoiceNode
    properties {defaultChoiceDisplayText=Continue, choiceTags {gameplayTags [{tagName=Athena.Quests.S29.Conversations.Snapshot.03.04}]}}
   }]
 }
 Speech_C_6 {
  id=#9
  class=Speech_C
  outputs []
  properties {generalConfig {
    defaultMessage="Once he's blown off some pent-up energy, he'll pick up the scent from you, and follow the trail."
    defaultMetadataParameters ["" {value=""}]
   }}
  sub_nodes [ConversationChoiceNode_4 {
    class=ConversationChoiceNode
    properties {defaultChoiceDisplayText=Continue, choiceTags {gameplayTags [{tagName=Athena.Quests.S29.Conversations.Snapshot.03.04}]}}
   }]
 }
 ConversationEntryPointNode_1 {
  id=#1
  class=ConversationEntryPointNode
  outputs [#10]
  properties {entryTag {tagName=Conversation.Entry.Season29.Snapshot.P03.Q01}}
 }
 Speech_C_7 {
  id=#10
  class=Speech_C
  outputs [#11]
  properties {
   softTaskIcon=&T_UI_NPC_MainQuest
   generalConfig {defaultMessage=Zeus has asked me to dig up information about your actions., defaultMetadataParameters [Cerberus {value=""}]}
  }
  sub_nodes [
   ConversationChoiceNode_8 {
    class=ConversationChoiceNode
    properties {
     defaultChoiceDisplayText=Speak with Cerberus
     choiceTags {gameplayTags [{tagName=Conversation.Choice.Category.MainQuestline}, {tagName=Conversation.Choice.AutoSelectIfValid}]}
    }
   }
   FortConversationRequirement_ControllerRequirement_0 {
    class=FortConversationRequirement_ControllerRequirement
    properties {
     nodeName=Require: Has Completed Quest
     requirement {
      class=FortControllerRequirement_HasCompletedQuest
      properties {bInverseCondition=true, questDefinition=/BattlePassS29_Quests/Items/QuestItems/Snapshots/Part03/quest_S29_snapshot_p03_q01}
     }
    }
   }
   FortConversationRequirement_ControllerRequirement_1 {
    class=FortConversationRequirement_ControllerRequirement
    properties {
     nodeName=Require: Is Quest Active
     requirement {
      class=FortControllerRequirement_IsQuestActive
      properties {questDefinition=&quest_S29_snapshot_p03_q01}
     }
    }
   }
   FortConversationRequirement_ControllerRequirement_2 {
    class=FortConversationRequirement_ControllerRequirement
    properties {
     nodeName=Require: Has Completed Quest Objective
     requirement {
      class=FortControllerRequirement_HasCompletedQuestObjective
      properties {
       bInverseCondition=true
       objectiveBackendName=quest_S29_snapshot_p03_q01_obj0
       questDefinition=&quest_S29_snapshot_p03_q01
      }
     }
    }
   }
  ]
 }
 Speech_C_0 {
  id=#11
  class=Speech_C
  outputs [#12]
  properties {generalConfig {defaultMessage="A command I am willing to ignore... If you help me. "}}
  sub_nodes [ConversationChoiceNode_3 {class=ConversationChoiceNode, properties {defaultChoiceDisplayText=Continue}}]
 }
 Speech_C_1 {
  id=#12
  class=Speech_C
  outputs [#13]
  properties {generalConfig {defaultMessage="My master entrusted me with a powerful artifact. And unfortunately, it was... misplaced...."}}
  sub_nodes [ConversationChoiceNode_0 {class=ConversationChoiceNode, properties {defaultChoiceDisplayText=Continue}}]
 }
 Speech_C_5 {
  id=#13
  class=Speech_C
  outputs []
  properties {generalConfig {defaultMessage="Be a good dog and fetch my artifact! If you fail... I will tell Zeus what you've been up to."}}
  sub_nodes [ConversationChoiceNode_1 {
    class=ConversationChoiceNode
    properties {defaultChoiceDisplayText=Continue, choiceTags {gameplayTags [{tagName=Athena.Quests.S29.Conversations.Snapshot.03.01}]}}
   }]
 }
]
