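# Conversation database for Season 29 Battle Pass Snapshot Part 4 quests, holding Medusa and Oracle dialogue for Q01, Q03, Q05, Q07 and Q09.
# Each entry tag leads to Speech chains gated on its quest being active, with quest and first objective incomplete.
# Quest-specific choice tags sit on auto-selected openers, later lines or both.
# Q01 shows a rebuff line instead while any listed Part 1–3 quest is active.
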ConversationDatabase /BattlePassS29_Quests/Conversations/Conversation_S29_Snapshot_P04
entry_points [
 {tag=Conversation.Entry.Season29.Snapshot.P04.Q01, destinations [#1]}
 {tag=Conversation.Entry.Season29.Snapshot.P04.Q05, destinations [#2]}
 {tag=Conversation.Entry.Season29.Snapshot.P04.Q03, destinations [#3]}
 {tag=Conversation.Entry.Season29.Snapshot.P04.Q09, destinations [#4]}
 {tag=Conversation.Entry.Season29.Snapshot.P04.Q07, destinations [#5]}
]
nodes [
 ConversationEntryPointNode_1 {
  id=#5
  class=ConversationEntryPointNode
  outputs [#6]
  properties {entryTag {tagName=Conversation.Entry.Season29.Snapshot.P04.Q07}}
 }
 Speech_C_4 {
  id=#6
  class=Speech_C
  outputs [#7]
  properties {
   softTaskIcon=&T_UI_NPC_MainQuest
   generalConfig {defaultMessage=These hourglasses are forged from Zeus' own godly essence., defaultMetadataParameters [Medusa {value=""}]}
  }
  sub_nodes [
   ConversationChoiceNode_11 {
    class=ConversationChoiceNode
    properties {
     defaultChoiceDisplayText=Talk to Medusa
     choiceTags {gameplayTags [
       {tagName=Conversation.Choice.AutoSelectIfValid}
       {tagName=Conversation.Choice.Category.MainQuestline}
       {tagName=Athena.Quests.S29.Conversations.Snapshot.04.06}
      ]}
    }
   }
   FortConversationRequirement_ControllerRequirement_20 {
    class=FortConversationRequirement_ControllerRequirement
    properties {
     nodeName=Require: Has Completed Quest
     requirement {
      class=FortControllerRequirement_HasCompletedQuest
      properties {bInverseCondition=true, questDefinition=/BattlePassS29_Quests/Items/QuestItems/Snapshots/Part04/quest_S29_snapshot_p04_q07}
     }
    }
   }
   FortConversationRequirement_ControllerRequirement_21 {
    class=FortConversationRequirement_ControllerRequirement
    properties {
     nodeName=Require: Is Quest Active
     requirement {
      class=FortControllerRequirement_IsQuestActive
      properties {questDefinition=&quest_S29_snapshot_p04_q07}
     }
    }
   }
   FortConversationRequirement_ControllerRequirement_22 {
    class=FortConversationRequirement_ControllerRequirement
    properties {
     nodeName=Require: Has Completed Quest Objective
     requirement {
      class=FortControllerRequirement_HasCompletedQuestObjective
      properties {
       bInverseCondition=true
       objectiveBackendName=quest_S29_snapshot_p04_q07_obj0
       questDefinition=&quest_S29_snapshot_p04_q07
      }
     }
    }
   }
  ]
 }
 Speech_C_7 {
  id=#7
  class=Speech_C
  outputs [#8]
  properties {generalConfig {defaultMessage=A dangerous bet he has made to increase his own power. Foolish.}}
  sub_nodes [ConversationChoiceNode_3 {class=ConversationChoiceNode, properties {defaultChoiceDisplayText=Continue}}]
 }
 Speech_C_8 {
  id=#8
  class=Speech_C
  outputs [#9]
  properties {generalConfig {defaultMessage=Take this charm I've created from the essence of the hourglasses you destroyed.}}
  sub_nodes [ConversationChoiceNode_15 {class=ConversationChoiceNode, properties {defaultChoiceDisplayText=Continue}}]
 }
 Speech_C_19 {
  id=#9
  class=Speech_C
  outputs []
  properties {
   softTaskIcon=/Game/Athena/HUD/Interaction/T_UI_NPC_MedusaConversation
   generalConfig {defaultMessage="Find and destroy Zeus's Stormforged Hourglass, and protect mortalkind once and for all."}
  }
  sub_nodes [ConversationChoiceNode_16 {
    class=ConversationChoiceNode
    properties {defaultChoiceDisplayText=Accept Medusa's Insignia, choiceTags {gameplayTags [{tagName=Athena.Quests.S29.Conversations.Snapshot.04.06}]}}
   }]
 }
 ConversationEntryPointNode_2 {
  id=#4
  class=ConversationEntryPointNode
  outputs [#10]
  properties {entryTag {tagName=Conversation.Entry.Season29.Snapshot.P04.Q09}}
 }
 Speech_C_6 {
  id=#10
  class=Speech_C
  outputs []
  properties {softTaskIcon=&T_UI_NPC_MainQuest, generalConfig {defaultMessage=You have won a great victory for humanity.}}
  sub_nodes [
   ConversationChoiceNode_9 {
    class=ConversationChoiceNode
    properties {
     defaultChoiceDisplayText=Talk to the Oracle
     choiceTags {gameplayTags [
       {tagName=Conversation.Choice.AutoSelectIfValid}
       {tagName=Conversation.Choice.Category.MainQuestline}
       {tagName=Athena.Quests.S29.Conversations.Snapshot.04.09}
      ]}
    }
   }
   FortConversationRequirement_ControllerRequirement_3 {
    class=FortConversationRequirement_ControllerRequirement
    properties {
     nodeName=Require: Has Completed Quest
     requirement {
      class=FortControllerRequirement_HasCompletedQuest
      properties {bInverseCondition=true, questDefinition=/BattlePassS29_Quests/Items/QuestItems/Snapshots/Part04/quest_S29_snapshot_p04_q09}
     }
    }
   }
   FortConversationRequirement_ControllerRequirement_4 {
    class=FortConversationRequirement_ControllerRequirement
    properties {
     nodeName=Require: Is Quest Active
     requirement {
      class=FortControllerRequirement_IsQuestActive
      properties {questDefinition=&quest_S29_snapshot_p04_q09}
     }
    }
   }
   FortConversationRequirement_ControllerRequirement_5 {
    class=FortConversationRequirement_ControllerRequirement
    properties {
     nodeName=Require: Has Completed Quest Objective
     requirement {
      class=FortControllerRequirement_HasCompletedQuestObjective
      properties {
       bInverseCondition=true
       objectiveBackendName=quest_S29_snapshot_p04_q09_obj0
       questDefinition=&quest_S29_snapshot_p04_q09
      }
     }
    }
   }
  ]
 }
 ConversationEntryPointNode_5 {
  id=#3
  class=ConversationEntryPointNode
  outputs [#11]
  properties {entryTag {tagName=Conversation.Entry.Season29.Snapshot.P04.Q03}}
 }
 Speech_C_9 {
  id=#11
  class=Speech_C
  outputs [#12]
  properties {
   softTaskIcon=&T_UI_NPC_MainQuest
   generalConfig {defaultMessage="This hourglass... it is not mine, but it resonates with godly power.", defaultMetadataParameters [Medusa {value=""}]}
  }
  sub_nodes [
   ConversationChoiceNode_0 {
    class=ConversationChoiceNode
    properties {
     defaultChoiceDisplayText=Talk to Medusa
     choiceTags {gameplayTags [
       {tagName=Conversation.Choice.AutoSelectIfValid}
       {tagName=Conversation.Choice.Category.MainQuestline}
       {tagName=Athena.Quests.S29.Conversations.Snapshot.04.03}
      ]}
    }
   }
   FortConversationRequirement_ControllerRequirement_1 {
    class=FortConversationRequirement_ControllerRequirement
    properties {
     nodeName=Require: Has Completed Quest
     requirement {
      class=FortControllerRequirement_HasCompletedQuest
      properties {bInverseCondition=true, questDefinition=/BattlePassS29_Quests/Items/QuestItems/Snapshots/Part04/quest_S29_snapshot_p04_q03}
     }
    }
   }
   FortConversationRequirement_ControllerRequirement_2 {
    class=FortConversationRequirement_ControllerRequirement
    properties {
     nodeName=Require: Is Quest Active
     requirement {
      class=FortControllerRequirement_IsQuestActive
      properties {questDefinition=&quest_S29_snapshot_p04_q03}
     }
    }
   }
   FortConversationRequirement_ControllerRequirement_6 {
    class=FortConversationRequirement_ControllerRequirement
    properties {
     nodeName=Require: Has Completed Quest Objective
     requirement {
      class=FortControllerRequirement_HasCompletedQuestObjective
      properties {
       bInverseCondition=true
       objectiveBackendName=quest_S29_snapshot_p04_q03_obj0
       questDefinition=&quest_S29_snapshot_p04_q03
      }
     }
    }
   }
  ]
 }
 Speech_C_10 {
  id=#12
  class=Speech_C
  outputs [#13]
  properties {generalConfig {defaultMessage="Zeus's power, to be specific. What is he up to?"}}
  sub_nodes [ConversationChoiceNode_1 {
    class=ConversationChoiceNode
    properties {defaultChoiceDisplayText=Continue, choiceTags {gameplayTags [{tagName=Athena.Quests.S29.Conversations.Snapshot.04.03}]}}
   }]
 }
 Speech_C_3 {
  id=#13
  class=Speech_C
  outputs [#14]
  properties {generalConfig {defaultMessage="While I investigate this, keep the gods and their minions distracted."}}
  sub_nodes [ConversationChoiceNode_2 {
    class=ConversationChoiceNode
    properties {defaultChoiceDisplayText=Continue, choiceTags {gameplayTags [{tagName=Athena.Quests.S29.Conversations.Snapshot.04.03}]}}
   }]
 }
 Speech_C_5 {
  id=#14
  class=Speech_C
  outputs []
  properties {generalConfig {defaultMessage="This might be the key to defeating Zeus, once and for all."}}
  sub_nodes [ConversationChoiceNode_7 {
    class=ConversationChoiceNode
    properties {defaultChoiceDisplayText=Continue, choiceTags {gameplayTags [{tagName=Athena.Quests.S29.Conversations.Snapshot.04.03}]}}
   }]
 }
 ConversationEntryPointNode_7 {
  id=#2
  class=ConversationEntryPointNode
  outputs [#15]
  properties {entryTag {tagName=Conversation.Entry.Season29.Snapshot.P04.Q05}}
 }
 Speech_C_18 {
  id=#15
  class=Speech_C
  outputs []
  properties {softTaskIcon=&T_UI_NPC_MainQuest, generalConfig {defaultMessage=I have seen the truth.}}
  sub_nodes [
   ConversationChoiceNode_20 {
    class=ConversationChoiceNode
    properties {
     defaultChoiceDisplayText=Speak with the Oracle
     choiceTags {gameplayTags [
       {tagName=Conversation.Choice.AutoSelectIfValid}
       {tagName=Conversation.Choice.Category.MainQuestline}
       {tagName=Athena.Quests.S29.Conversations.Snapshot.04.05}
      ]}
    }
   }
   FortConversationRequirement_ControllerRequirement_15 {
    class=FortConversationRequirement_ControllerRequirement
    properties {
     nodeName=Require: Has Completed Quest
     requirement {
      class=FortControllerRequirement_HasCompletedQuest
      properties {bInverseCondition=true, questDefinition=/BattlePassS29_Quests/Items/QuestItems/Snapshots/Part04/quest_S29_snapshot_p04_q05}
     }
    }
   }
   FortConversationRequirement_ControllerRequirement_16 {
    class=FortConversationRequirement_ControllerRequirement
    properties {
     nodeName=Require: Is Quest Active
     requirement {
      class=FortControllerRequirement_IsQuestActive
      properties {questDefinition=&quest_S29_snapshot_p04_q05}
     }
    }
   }
   FortConversationRequirement_ControllerRequirement_17 {
    class=FortConversationRequirement_ControllerRequirement
    properties {
     nodeName=Require: Has Completed Quest Objective
     requirement {
      class=FortControllerRequirement_HasCompletedQuestObjective
      properties {
       bInverseCondition=true
       objectiveBackendName=quest_S29_snapshot_p04_q05_obj0
       questDefinition=&quest_S29_snapshot_p04_q05
      }
     }
    }
   }
  ]
 }
 ConversationEntryPointNode_0 {
  id=#1
  class=ConversationEntryPointNode
  outputs [#16, #17]
  properties {entryTag {tagName=Conversation.Entry.Season29.Snapshot.P04.Q01}}
 }
 Speech_C_12 {
  id=#17
  class=Speech_C
  outputs []
  properties {generalConfig {defaultMessage="Come to me when you've properly proven yourself, \"hero\".", defaultMetadataParameters [Medusa {value=""}]}}
  sub_nodes [FortConversationRequirement_ControllerRequirement_11 {
    class=FortConversationRequirement_ControllerRequirement
    properties {
     nodeName=Require: Is Quest Active
     requirement {
      class=FortControllerRequirement_IsQuestActive
      properties {
       questDefinition=/BattlePassS29_Quests/Items/QuestItems/Snapshots/Part01/quest_S29_snapshot_p01_q14
       testPolicy=AnyOf
       questDefinitions [
        /BattlePassS29_Quests/Items/QuestItems/Snapshots/Part01/quest_S29_snapshot_p01_q01
        /BattlePassS29_Quests/Items/QuestItems/Snapshots/Part01/quest_S29_snapshot_p01_q02
        /BattlePassS29_Quests/Items/QuestItems/Snapshots/Part01/quest_S29_snapshot_p01_q03
        /BattlePassS29_Quests/Items/QuestItems/Snapshots/Part01/quest_S29_snapshot_p01_q04
        /BattlePassS29_Quests/Items/QuestItems/Snapshots/Part01/quest_S29_snapshot_p01_q05
        /BattlePassS29_Quests/Items/QuestItems/Snapshots/Part01/quest_S29_snapshot_p01_q06
        /BattlePassS29_Quests/Items/QuestItems/Snapshots/Part01/quest_S29_snapshot_p01_q07
        /BattlePassS29_Quests/Items/QuestItems/Snapshots/Part01/quest_S29_snapshot_p01_q08
        /BattlePassS29_Quests/Items/QuestItems/Snapshots/Part01/quest_S29_snapshot_p01_q09
        /BattlePassS29_Quests/Items/QuestItems/Snapshots/Part01/quest_S29_snapshot_p01_q10
        /BattlePassS29_Quests/Items/QuestItems/Snapshots/Part01/quest_S29_snapshot_p01_q11
        /BattlePassS29_Quests/Items/QuestItems/Snapshots/Part01/quest_S29_snapshot_p01_q12
        /BattlePassS29_Quests/Items/QuestItems/Snapshots/Part01/quest_S29_snapshot_p01_q13
        &quest_S29_snapshot_p01_q14
        /BattlePassS29_Quests/Items/QuestItems/Snapshots/Part02/quest_S29_snapshot_p02_q01
        /BattlePassS29_Quests/Items/QuestItems/Snapshots/Part02/quest_S29_snapshot_p02_q02
        /BattlePassS29_Quests/Items/QuestItems/Snapshots/Part02/quest_S29_snapshot_p02_q03
        /BattlePassS29_Quests/Items/QuestItems/Snapshots/Part02/quest_S29_snapshot_p02_q04
        /BattlePassS29_Quests/Items/QuestItems/Snapshots/Part02/quest_S29_snapshot_p02_q05
        /BattlePassS29_Quests/Items/QuestItems/Snapshots/Part02/quest_S29_snapshot_p02_q06
        /BattlePassS29_Quests/Items/QuestItems/Snapshots/Part02/quest_S29_snapshot_p02_q07
        /BattlePassS29_Quests/Items/QuestItems/Snapshots/Part02/quest_S29_snapshot_p02_q08
        /BattlePassS29_Quests/Items/QuestItems/Snapshots/Part02/quest_S29_snapshot_p02_q09
        /BattlePassS29_Quests/Items/QuestItems/Snapshots/Part02/quest_S29_snapshot_p02_q10
        /BattlePassS29_Quests/Items/QuestItems/Snapshots/Part02/quest_S29_snapshot_p02_q11
        /BattlePassS29_Quests/Items/QuestItems/Snapshots/Part02/quest_S29_snapshot_p02_q12
        /BattlePassS29_Quests/Items/QuestItems/Snapshots/Part02/quest_S29_snapshot_p02_q04b
        /BattlePassS29_Quests/Items/QuestItems/Snapshots/Part02/quest_S29_snapshot_p02_q07b
        /BattlePassS29_Quests/Items/QuestItems/Snapshots/Part02/quest_S29_snapshot_p02_q10b
        /BattlePassS29_Quests/Items/QuestItems/Snapshots/Part02/quest_S29_snapshot_p02_q12b
        /BattlePassS29_Quests/Items/QuestItems/Snapshots/Part03/quest_S29_snapshot_p03_q01
        /BattlePassS29_Quests/Items/QuestItems/Snapshots/Part03/quest_S29_snapshot_p03_q02
        /BattlePassS29_Quests/Items/QuestItems/Snapshots/Part03/quest_S29_snapshot_p03_q04
        /BattlePassS29_Quests/Items/QuestItems/Snapshots/Part03/quest_S29_snapshot_p03_q05
        /BattlePassS29_Quests/Items/QuestItems/Snapshots/Part03/quest_S29_snapshot_p03_q06
        /BattlePassS29_Quests/Items/QuestItems/Snapshots/Part03/quest_S29_snapshot_p03_q07
        /BattlePassS29_Quests/Items/QuestItems/Snapshots/Part03/quest_S29_snapshot_p03_q08
        /BattlePassS29_Quests/Items/QuestItems/Snapshots/Part03/quest_S29_snapshot_p03_q03a_01
        /BattlePassS29_Quests/Items/QuestItems/Snapshots/Part03/quest_S29_snapshot_p03_q03a_02
        /BattlePassS29_Quests/Items/QuestItems/Snapshots/Part03/quest_s29_snapshot_p03_q03b_01
        /BattlePassS29_Quests/Items/QuestItems/Snapshots/Part03/quest_s29_snapshot_p03_q03b_02
        /BattlePassS29_Quests/Items/QuestItems/Snapshots/Part03/quest_s29_snapshot_p03_q03c_01
        /BattlePassS29_Quests/Items/QuestItems/Snapshots/Part03/quest_s29_snapshot_p03_q03c_02
        /BattlePassS29_Quests/Items/QuestItems/Snapshots/Part03/quest_S29_snapshot_p03_q04_granter
       ]
      }
     }
    }
   }]
 }
 Speech_C_13 {
  id=#16
  class=Speech_C
  outputs [#18]
  properties {
   softTaskIcon=&T_UI_NPC_MainQuest
   generalConfig {
    defaultMessage="I've always loathed Zeus... But his time grows short. Together, we will make it shorter."
    defaultMetadataParameters [Medusa {value=""}]
   }
  }
  sub_nodes [
   ConversationChoiceNode_6 {
    class=ConversationChoiceNode
    properties {
     defaultChoiceDisplayText=Speak to Medusa
     choiceTags {gameplayTags [{tagName=Conversation.Choice.AutoSelectIfValid}, {tagName=Conversation.Choice.Category.MainQuestline}]}
    }
   }
   FortConversationRequirement_ControllerRequirement_10 {
    class=FortConversationRequirement_ControllerRequirement
    properties {
     nodeName=Require: Is Quest Active
     requirement {
      class=FortControllerRequirement_IsQuestActive
      properties {
       bInverseCondition=true
       questDefinition=&quest_S29_snapshot_p01_q14
       testPolicy=AnyOf
       questDefinitions [
        &quest_S29_snapshot_p01_q01
        &quest_S29_snapshot_p01_q02
        &quest_S29_snapshot_p01_q03
        &quest_S29_snapshot_p01_q04
        &quest_S29_snapshot_p01_q05
        &quest_S29_snapshot_p01_q06
        &quest_S29_snapshot_p01_q07
        &quest_S29_snapshot_p01_q08
        &quest_S29_snapshot_p01_q09
        &quest_S29_snapshot_p01_q10
        &quest_S29_snapshot_p01_q11
        &quest_S29_snapshot_p01_q12
        &quest_S29_snapshot_p01_q13
        &quest_S29_snapshot_p01_q14
        &quest_S29_snapshot_p02_q01
        &quest_S29_snapshot_p02_q02
        &quest_S29_snapshot_p02_q03
        &quest_S29_snapshot_p02_q04
        &quest_S29_snapshot_p02_q05
        &quest_S29_snapshot_p02_q06
        &quest_S29_snapshot_p02_q07
        &quest_S29_snapshot_p02_q08
        &quest_S29_snapshot_p02_q09
        &quest_S29_snapshot_p02_q10
        &quest_S29_snapshot_p02_q11
        &quest_S29_snapshot_p02_q12
        &quest_S29_snapshot_p02_q04b
        &quest_S29_snapshot_p02_q07b
        &quest_S29_snapshot_p02_q10b
        &quest_S29_snapshot_p02_q12b
        &quest_S29_snapshot_p03_q01
        &quest_S29_snapshot_p03_q02
        &quest_S29_snapshot_p03_q04
        &quest_S29_snapshot_p03_q05
        &quest_S29_snapshot_p03_q06
        &quest_S29_snapshot_p03_q07
        &quest_S29_snapshot_p03_q08
        &quest_S29_snapshot_p03_q03a_01
        &quest_S29_snapshot_p03_q03a_02
        &quest_s29_snapshot_p03_q03b_01
        &quest_s29_snapshot_p03_q03b_02
        &quest_s29_snapshot_p03_q03c_01
        &quest_s29_snapshot_p03_q03c_02
        &quest_S29_snapshot_p03_q04_granter
       ]
      }
     }
    }
   }
   FortConversationRequirement_ControllerRequirement_7 {
    class=FortConversationRequirement_ControllerRequirement
    properties {
     nodeName=Require: Has Completed Quest
     requirement {
      class=FortControllerRequirement_HasCompletedQuest
      properties {bInverseCondition=true, questDefinition=/BattlePassS29_Quests/Items/QuestItems/Snapshots/Part04/quest_S29_snapshot_p04_q01}
     }
    }
   }
   FortConversationRequirement_ControllerRequirement_8 {
    class=FortConversationRequirement_ControllerRequirement
    properties {
     nodeName=Require: Is Quest Active
     requirement {
      class=FortControllerRequirement_IsQuestActive
      properties {questDefinition=&quest_S29_snapshot_p04_q01}
     }
    }
   }
   FortConversationRequirement_ControllerRequirement_9 {
    class=FortConversationRequirement_ControllerRequirement
    properties {
     nodeName=Require: Has Completed Quest Objective
     requirement {
      class=FortControllerRequirement_HasCompletedQuestObjective
      properties {
       bInverseCondition=true
       objectiveBackendName=quest_S29_snapshot_p04_q01_obj0
       questDefinition=&quest_S29_snapshot_p04_q01
      }
     }
    }
   }
  ]
 }
 Speech_C_0 {
  id=#18
  class=Speech_C
  outputs [#19]
  properties {generalConfig {defaultMessage=Over the years he's surrounded himself with powerful items. A gauche show of his absolute rule.}}
  sub_nodes [ConversationChoiceNode_4 {class=ConversationChoiceNode, properties {defaultChoiceDisplayText=Continue}}]
 }
 Speech_C_1 {
  id=#19
  class=Speech_C
  outputs [#20]
  properties {generalConfig {defaultMessage="Fortunately for us, they're not his items. Most of them were made by me. Stolen from me."}}
  sub_nodes [ConversationChoiceNode_12 {class=ConversationChoiceNode, properties {defaultChoiceDisplayText=Continue}}]
 }
 Speech_C_2 {
  id=#20
  class=Speech_C
  outputs []
  properties {generalConfig {defaultMessage="I think it's about time we reclaim my property, don't you?"}}
  sub_nodes [ConversationChoiceNode_13 {
    class=ConversationChoiceNode
    properties {defaultChoiceDisplayText=Continue, choiceTags {gameplayTags [{tagName=Athena.Quests.S29.Conversations.Snapshot.04.01}]}}
   }]
 }
]
