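# Conversation database for three Season 29 Seasonal Activation quest dialogues with Marigold, Brutus and Jules.
# Each entry tag leads to an auto-selected main-questline speech, gated on its quest being active with the objective incomplete.
# Marigold's dialogue is a three-question deduction of Midas's escape, with wrong answers routed through Back nodes; Brutus and Jules are linear.
# Quest conversation tags sit on closing choices, repeated across Marigold's last three.
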ConversationDatabase /BattlePassS29_Quests/Conversations/Conversation_SeasonalActivation
entry_points [
 {tag=Conversation.Entry.Season29.SeasonalActivation.01b, destinations [#1]}
 {tag=Conversation.Entry.Season29.SeasonalActivation.09, destinations [#2]}
 {tag=Conversation.Entry.Season29.SeasonalActivation.10c, destinations [#3]}
]
nodes [
 ConversationEntryPointNode_2 {
  id=#3
  class=ConversationEntryPointNode
  outputs [#4]
  properties {entryTag {tagName=Conversation.Entry.Season29.SeasonalActivation.10c}}
 }
 Speech_C_11 {
  id=#4
  class=Speech_C
  outputs [#5]
  properties {
   softTaskIcon=&T_UI_NPC_MainQuest
   generalConfig {
    defaultMessage=With this Essence and your observations on how the Chains of Hades affect the Golden Touch...
    defaultMetadataParameters [Jules {value=""}]
   }
  }
  sub_nodes [
   ConversationChoiceNode_44 {
    class=ConversationChoiceNode
    properties {
     defaultChoiceDisplayText=Give Essence to Jules
     choiceTags {gameplayTags [{tagName=Conversation.Choice.Category.MainQuestline}, {tagName=Conversation.Choice.AutoSelectIfValid}]}
    }
   }
   FortConversationRequirement_ControllerRequirement_85 {
    class=FortConversationRequirement_ControllerRequirement
    properties {
     nodeName=Require: Is Quest Active
     requirement {
      class=FortControllerRequirement_IsQuestActive
      properties {questDefinition=/BattlePassS29_Quests/Items/QuestItems/SpringActivation/Drop2/quest_S29_SA_D02_q10c}
     }
    }
   }
   FortConversationRequirement_ControllerRequirement_86 {
    class=FortConversationRequirement_ControllerRequirement
    properties {
     nodeName=Require: Has Completed Quest Objective
     requirement {
      class=FortControllerRequirement_HasCompletedQuestObjective
      properties {
       bInverseCondition=true
       objectiveBackendName=quest_S29_SA_D02_q10c_obj0
       questDefinition=&quest_S29_SA_D02_q10c
      }
     }
    }
   }
  ]
 }
 Speech_C_12 {
  id=#5
  class=Speech_C
  outputs [#6]
  properties {generalConfig {defaultMessage="I think I can create gauntlets for my dad that can suppress his Golden Touch. "}}
  sub_nodes [ConversationChoiceNode_45 {class=ConversationChoiceNode, properties {defaultChoiceDisplayText=Continue}}]
 }
 Speech_C_13 {
  id=#6
  class=Speech_C
  outputs [#7]
  properties {generalConfig {defaultMessage="Ones he can take off! Whenever he wants! This could change everything."}}
  sub_nodes [ConversationChoiceNode_10 {class=ConversationChoiceNode, properties {defaultChoiceDisplayText=Continue}}]
 }
 Speech_C_14 {
  id=#7
  class=Speech_C
  outputs []
  properties {generalConfig {defaultMessage="If this works... Thank you. I'm going to get to work!"}}
  sub_nodes [ConversationChoiceNode_5 {
    class=ConversationChoiceNode
    properties {defaultChoiceDisplayText=Continue, choiceTags {gameplayTags [{tagName=Athena.Quests.S29.Conversations.SeasonalActivation.10c}]}}
   }]
 }
 ConversationEntryPointNode_1 {
  id=#2
  class=ConversationEntryPointNode
  outputs [#8]
  properties {entryTag {tagName=Conversation.Entry.Season29.SeasonalActivation.09}}
 }
 Speech_C_3 {
  id=#8
  class=Speech_C
  outputs [#9]
  properties {
   softTaskIcon=&T_UI_NPC_MainQuest
   generalConfig {defaultMessage=I heard you've been sniffing around..., defaultMetadataParameters [Brutus {value=""}]}
  }
  sub_nodes [
   ConversationChoiceNode_6 {
    class=ConversationChoiceNode
    properties {
     defaultChoiceDisplayText=Confront Brutus
     choiceTags {gameplayTags [{tagName=Conversation.Choice.Category.MainQuestline}, {tagName=Conversation.Choice.AutoSelectIfValid}]}
    }
   }
   FortConversationRequirement_ControllerRequirement_4 {
    class=FortConversationRequirement_ControllerRequirement
    properties {
     nodeName=Require: Is Quest Active
     requirement {
      class=FortControllerRequirement_IsQuestActive
      properties {questDefinition=/BattlePassS29_Quests/Items/QuestItems/SpringActivation/Drop1/quest_S29_SA_D01_q09}
     }
    }
   }
   FortConversationRequirement_ControllerRequirement_5 {
    class=FortConversationRequirement_ControllerRequirement
    properties {
     nodeName=Require: Has Completed Quest Objective
     requirement {
      class=FortControllerRequirement_HasCompletedQuestObjective
      properties {
       bInverseCondition=true
       objectiveBackendName=quest_S29_SA_D01_q09_obj0
       questDefinition=&quest_S29_SA_D01_q09
      }
     }
    }
   }
  ]
 }
 Speech_C_4 {
  id=#9
  class=Speech_C
  outputs [#10]
  properties {generalConfig {defaultMessage="Good! I'm glad the boss has another person watching his back."}}
  sub_nodes [ConversationChoiceNode_7 {class=ConversationChoiceNode, properties {defaultChoiceDisplayText=Continue}}]
 }
 Speech_C_16 {
  id=#10
  class=Speech_C
  outputs [#11]
  properties {generalConfig {defaultMessage="Those bugs? Aw, those are Hope's, from when The Marigold was a Society boat. I'll get 'em back to her."}}
  sub_nodes [ConversationChoiceNode_9 {class=ConversationChoiceNode, properties {defaultChoiceDisplayText=Continue}}]
 }
 Speech_C_9 {
  id=#11
  class=Speech_C
  outputs [#12]
  properties {generalConfig {defaultMessage="Honestly, I'm really looking forward to going back to being number 2."}}
  sub_nodes [ConversationChoiceNode_11 {class=ConversationChoiceNode, properties {defaultChoiceDisplayText=Continue}}]
 }
 Speech_C_5 {
  id=#12
  class=Speech_C
  outputs []
  properties {generalConfig {defaultMessage="I love this crew, but managing a bunch of big personalities is a lot of stress."}}
  sub_nodes [ConversationChoiceNode_8 {
    class=ConversationChoiceNode
    properties {defaultChoiceDisplayText=Continue, choiceTags {gameplayTags [{tagName=Athena.Quests.S29.Conversations.SeasonalActivation.09}]}}
   }]
 }
 ConversationEntryPointNode_10 {
  id=#1
  class=ConversationEntryPointNode
  outputs [#13]
  properties {entryTag {tagName=Conversation.Entry.Season29.SeasonalActivation.01b}}
 }
 Speech_C_2 {
  id=#13
  class=Speech_C
  outputs [#14]
  properties {
   softTaskIcon=&T_UI_NPC_MainQuest
   generalConfig {defaultMessage=It looks like Midas escaped. He's always been clever. Sometimes too clever., defaultMetadataParameters [Marigold {value=""}]}
  }
  sub_nodes [
   ConversationChoiceNode_36 {
    class=ConversationChoiceNode
    properties {
     defaultChoiceDisplayText=Corroborate with Marigold
     choiceTags {gameplayTags [{tagName=Conversation.Choice.AutoSelectIfValid}, {tagName=Conversation.Choice.Category.MainQuestline}]}
    }
   }
   FortConversationRequirement_ControllerRequirement_69 {
    class=FortConversationRequirement_ControllerRequirement
    properties {
     nodeName=Require: Is Quest Active
     requirement {
      class=FortControllerRequirement_IsQuestActive
      properties {questDefinition=/BattlePassS29_Quests/Items/QuestItems/SpringActivation/Drop0/Quest_S29_SA_D00_q01b}
     }
    }
   }
   FortConversationRequirement_ControllerRequirement_70 {
    class=FortConversationRequirement_ControllerRequirement
    properties {
     nodeName=Require: Has Completed Quest Objective
     requirement {
      class=FortControllerRequirement_HasCompletedQuestObjective
      properties {
       bInverseCondition=true
       objectiveBackendName=quest_S29_SA_D00_q01b_obj0
       questDefinition=&Quest_S29_SA_D00_q01b
      }
     }
    }
   }
  ]
 }
 Speech_C_36 {
  id=#14
  class=Speech_C
  outputs [#15, #16, #17]
  properties {generalConfig {defaultMessage=It looks as though those Chains of Hades were able to suppress his Golden Touch. So how did he escape?}}
  sub_nodes [ConversationChoiceNode_22 {class=ConversationChoiceNode, properties {defaultChoiceDisplayText=Continue}}]
 }
 Speech_C_15 {
  id=#17
  class=Speech_C
  outputs [#18]
  properties {generalConfig {defaultMessage=And he turned them to gold. Which means his curse is only suppressed while the cuff is on.}}
  sub_nodes [ConversationChoiceNode_18 {class=ConversationChoiceNode, properties {defaultChoiceDisplayText=A guard unshackled one of his hands??}}]
 }
 Speech_C_35 {
  id=#18
  class=Speech_C
  outputs [#19, #20, #21]
  properties {generalConfig {defaultMessage=Interesting. Then what happened?}}
  sub_nodes [ConversationChoiceNode_29 {class=ConversationChoiceNode, properties {defaultChoiceDisplayText=Continue}}]
 }
 Speech_C_24 {
  id=#21
  class=Speech_C
  outputs [#22]
  properties {generalConfig {defaultMessage="Me? No, I arrived here after I heard he returned. Look again."}}
  sub_nodes [ConversationChoiceNode_14 {class=ConversationChoiceNode, properties {defaultChoiceDisplayText="You helped him escape!"}}]
 }
 Back_C_0 {id=#22, class=Back_C}
 Speech_C_23 {
  id=#20
  class=Speech_C
  outputs [#22]
  properties {generalConfig {defaultMessage=Take another look. The answer's right in front of you.}}
  sub_nodes [ConversationChoiceNode_13 {class=ConversationChoiceNode, properties {defaultChoiceDisplayText="He slipped his other hand out of the restraints!"}}]
 }
 Speech_C_25 {
  id=#19
  class=Speech_C
  outputs [#23]
  properties {generalConfig {defaultMessage="Yes. That softened the metal, allowing him to break the shackle and free his other hand."}}
  sub_nodes [ConversationChoiceNode_15 {class=ConversationChoiceNode, properties {defaultChoiceDisplayText="He used his free hand to turn the other shackle gold!"}}]
 }
 Speech_C_29 {
  id=#23
  class=Speech_C
  outputs [#24, #25, #26]
  properties {generalConfig {defaultMessage=And what was the final step to his escape?}}
  sub_nodes [ConversationChoiceNode_37 {class=ConversationChoiceNode, properties {defaultChoiceDisplayText=Continue}}]
 }
 Speech_C_28 {
  id=#26
  class=Speech_C
  outputs [#27]
  properties {generalConfig {defaultMessage="Indeed. He touched the cell door to soften the lock, and pushed it open."}}
  sub_nodes [ConversationChoiceNode_21 {
    class=ConversationChoiceNode
    properties {
     defaultChoiceDisplayText=He shoved open the cell door?
     choiceTags {gameplayTags [{tagName=Athena.Quests.S29.Conversations.SeasonalActivation.01b}]}
    }
   }]
 }
 Speech_C_32 {
  id=#27
  class=Speech_C
  outputs [#28]
  properties {generalConfig {defaultMessage="Then he walked out of The Underworld, free at last."}}
  sub_nodes [ConversationChoiceNode_26 {
    class=ConversationChoiceNode
    properties {defaultChoiceDisplayText=Continue, choiceTags {gameplayTags [{tagName=Athena.Quests.S29.Conversations.SeasonalActivation.01b}]}}
   }]
 }
 Speech_C_33 {
  id=#28
  class=Speech_C
  outputs []
  properties {generalConfig {defaultMessage="Now we know how those Chains of Hades interact with the Golden Touch. That will be useful, I'm sure."}}
  sub_nodes [ConversationChoiceNode_24 {
    class=ConversationChoiceNode
    properties {defaultChoiceDisplayText=Continue, choiceTags {gameplayTags [{tagName=Athena.Quests.S29.Conversations.SeasonalActivation.01b}]}}
   }]
 }
 Speech_C_27 {
  id=#25
  class=Speech_C
  outputs [#29]
  properties {generalConfig {defaultMessage=I don't think so. The scene is telling a different story.}}
  sub_nodes [ConversationChoiceNode_20 {class=ConversationChoiceNode, properties {defaultChoiceDisplayText=He transformed into a bat?}}]
 }
 Back_C_1 {id=#29, class=Back_C}
 Speech_C_26 {
  id=#24
  class=Speech_C
  outputs [#29]
  properties {generalConfig {defaultMessage=I don't think so. The scene is telling a different story.}}
  sub_nodes [ConversationChoiceNode_19 {class=ConversationChoiceNode, properties {defaultChoiceDisplayText=Zeus broke him out}}]
 }
 Speech_C_10 {
  id=#16
  class=Speech_C
  outputs [#30]
  properties {generalConfig {defaultMessage="Hm, I don't think that's it. Take another look at the cell."}}
  sub_nodes [ConversationChoiceNode_0 {class=ConversationChoiceNode, properties {defaultChoiceDisplayText=He somehow rifted out??}}]
 }
 Back_C_22 {id=#30, class=Back_C}
 Speech_C_8 {
  id=#15
  class=Speech_C
  outputs [#30]
  properties {generalConfig {defaultMessage="Hm, I don't think that's it. Take another look at the cell."}}
  sub_nodes [ConversationChoiceNode_16 {class=ConversationChoiceNode, properties {defaultChoiceDisplayText=He escaped through a hidden tunnel??}}]
 }
]
