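# Season 29 Oracle conversation hub for the Sun Rose Pool, routing players to quest dialogue through link nodes.
# Specific active snapshot quests unlock dedicated opening lines, most tagged to auto-select, that jump to their conversations.
# Otherwise, a greeting, also blocked during quest VO, leads to a quest-gated contact menu of characters with icons and voice cues.
# A contextual "check back later" message replaces the menu prompt when no contact quest is active.
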ConversationDatabase /BattlePassS29_Quests/Conversations/Conversation_SunRosePool
entry_points [{tag=Conversation.Entry.Season29.SunRosePool, destinations [#1]}]
nodes [
 ConversationEntryPointNode_1 {
  id=#1
  class=ConversationEntryPointNode
  outputs [
   #2
   #3
   #4
   #5
   #6
   #7
   #8
  ]
  properties {entryTag {tagName=Conversation.Entry.Season29.SunRosePool}}
 }
 Speech_C_8 {
  id=#8
  class=Speech_C
  outputs [#9]
  properties {generalConfig {defaultMessage=You have won a great victory for humanity., defaultMetadataParameters [Oracle Convo Start {value=""}]}}
  sub_nodes [
   ConversationChoiceNode_21 {
    class=ConversationChoiceNode
    properties {
     defaultChoiceDisplayText=Continue
     choiceTags {gameplayTags [
       {tagName=Conversation.Choice.Category.MainQuestline}
       {tagName=Conversation.Choice.AutoSelectIfValid}
       {tagName=Athena.Quests.S29.Conversations.Snapshot.04.09}
      ]}
    }
   }
   FortConversationRequirement_ControllerRequirement_7 {
    class=FortConversationRequirement_ControllerRequirement
    properties {
     nodeName=Require: Is Quest Active
     requirement {
      class=FortControllerRequirement_IsQuestActive
      properties {
       questDefinition=/BattlePassS29_Quests/Items/QuestItems/Snapshots/Part04/quest_S29_snapshot_p04_q09
       questDefinitions [&quest_S29_snapshot_p04_q09]
      }
     }
    }
   }
  ]
 }
 ConversationLinkNode_25 {
  id=#9
  class=ConversationLinkNode
  properties {remoteEntryTag {tagName=Conversation.Entry.Season29.Snapshot.P04.Q09}}
 }
 Speech_C_7 {
  id=#7
  class=Speech_C
  outputs [#10]
  properties {generalConfig {defaultMessage="I have seen the truth. ", defaultMetadataParameters [Oracle Convo Start {value=""}]}}
  sub_nodes [
   ConversationChoiceNode_3 {
    class=ConversationChoiceNode
    properties {
     defaultChoiceDisplayText=Continue
     choiceTags {gameplayTags [
       {tagName=Conversation.Choice.Category.MainQuestline}
       {tagName=Conversation.Choice.AutoSelectIfValid}
       {tagName=Athena.Quests.S29.Conversations.Snapshot.04.05}
      ]}
    }
   }
   FortConversationRequirement_ControllerRequirement_8 {
    class=FortConversationRequirement_ControllerRequirement
    properties {
     nodeName=Require: Is Quest Active
     requirement {
      class=FortControllerRequirement_IsQuestActive
      properties {
       questDefinition=/BattlePassS29_Quests/Items/QuestItems/Snapshots/Part04/quest_S29_snapshot_p04_q05
       questDefinitions [&quest_S29_snapshot_p04_q05]
      }
     }
    }
   }
  ]
 }
 ConversationLinkNode_26 {
  id=#10
  class=ConversationLinkNode
  properties {remoteEntryTag {tagName=Conversation.Entry.Season29.Snapshot.P04.Q05}}
 }
 Speech_C_3 {
  id=#6
  class=Speech_C
  outputs [#11]
  properties {generalConfig {defaultMessage="Zeus has many enemies, some who would welcome a fight...", defaultMetadataParameters [Oracle Convo Start {value=""}]}}
  sub_nodes [
   ConversationChoiceNode_1 {
    class=ConversationChoiceNode
    properties {
     defaultChoiceDisplayText=Continue
     choiceTags {gameplayTags [
       {tagName=Conversation.Choice.Category.MainQuestline}
       {tagName=Conversation.Choice.AutoSelectIfValid}
       {tagName=Athena.Quests.S29.Conversations.Snapshot.03.08}
      ]}
    }
   }
   FortConversationRequirement_ControllerRequirement_3 {
    class=FortConversationRequirement_ControllerRequirement
    properties {
     nodeName=Require: Is Quest Active
     requirement {
      class=FortControllerRequirement_IsQuestActive
      properties {questDefinition=/BattlePassS29_Quests/Items/QuestItems/Snapshots/Part03/quest_S29_snapshot_p03_q08}
     }
    }
   }
  ]
 }
 ConversationLinkNode_2 {
  id=#11
  class=ConversationLinkNode
  properties {remoteEntryTag {tagName=Conversation.Entry.Season29.Snapshot.P03.Q08}}
 }
 Speech_C_27 {
  id=#5
  class=Speech_C
  outputs [#12]
  properties {generalConfig {
    defaultMessage=Zeus's closest allies have seen you fight. And you're stronger than they thought.
    defaultMetadataParameters [Oracle Convo Start {value=""}]
   }}
  sub_nodes [
   ConversationChoiceNode_0 {
    class=ConversationChoiceNode
    properties {
     defaultChoiceDisplayText=Continue
     choiceTags {gameplayTags [
       {tagName=Athena.Quests.S29.Conversations.Snapshot.P01.Choice.Q14}
       {tagName=Conversation.Choice.Category.MainQuestline}
       {tagName=Conversation.Choice.AutoSelectIfValid}
      ]}
    }
   }
   FortConversationRequirement_ControllerRequirement_2 {
    class=FortConversationRequirement_ControllerRequirement
    properties {
     nodeName=Require: Is Quest Active
     requirement {
      class=FortControllerRequirement_IsQuestActive
      properties {questDefinition=/BattlePassS29_Quests/Items/QuestItems/Snapshots/Part01/quest_S29_snapshot_p01_q14}
     }
    }
   }
  ]
 }
 ConversationLinkNode_24 {
  id=#12
  class=ConversationLinkNode
  properties {remoteEntryTag {tagName=Conversation.Entry.Season29.Snapshot.P01.Q14}}
 }
 Speech_C_30 {
  id=#4
  class=Speech_C
  outputs [#13]
  properties {generalConfig {defaultMessage=Hero. Welcome., defaultMetadataParameters [Oracle Convo Start {value=""}]}}
  sub_nodes [
   ConversationChoiceNode_24 {class=ConversationChoiceNode, properties {defaultChoiceDisplayText=Continue}}
   SideEffect_Client_PlayNPCSoundCueEvent_C_5 {
    class=SideEffect_Client_PlayNPCSoundCueEvent_C
    properties {soundCueToPlay=/BattlePassS29_Quests/Audio/VO/Cues/VO_ORACLE_S29_2023_00_100_Dialogue_Cue}
   }
   FortConversationRequirement_ControllerRequirement_39 {
    class=FortConversationRequirement_ControllerRequirement
    properties {
     nodeName=Require: Is Quest Active
     requirement {
      class=FortControllerRequirement_IsQuestActive
      properties {questDefinition=/BattlePassS29_Quests/Items/QuestItems/Snapshots/Part01/quest_S29_snapshot_p01_q04}
     }
    }
   }
  ]
 }
 ConversationLinkNode_27 {
  id=#13
  class=ConversationLinkNode
  properties {remoteEntryTag {tagName=Conversation.Entry.Season29.Snapshot.P01.Q04}}
 }
 Speech_C_23 {
  id=#3
  class=Speech_C
  outputs [#14]
  properties {generalConfig {defaultMessage="Zeus' wrath is upon you. ", defaultMetadataParameters [Oracle Convo Start {value=""}]}}
  sub_nodes [
   ConversationChoiceNode_14 {
    class=ConversationChoiceNode
    properties {
     defaultChoiceDisplayText=Continue
     choiceTags {gameplayTags [{tagName=Athena.Quests.S29.Conversations.Snapshot.P01.Choice.Q01}, {tagName=Conversation.Choice.AutoSelectIfValid}]}
    }
   }
   FortConversationRequirement_ControllerRequirement_27 {
    class=FortConversationRequirement_ControllerRequirement
    properties {
     nodeName=Require: Is Quest Active
     requirement {
      class=FortControllerRequirement_IsQuestActive
      properties {
       questDefinition=/BattlePassS29_Quests/Items/QuestItems/Snapshots/Part01/quest_S29_snapshot_p01_q01
       questDefinitions [&quest_S29_snapshot_p01_q01]
      }
     }
    }
   }
  ]
 }
 ConversationLinkNode_12 {
  id=#14
  class=ConversationLinkNode
  properties {remoteEntryTag {tagName=Conversation.Entry.Season29.Snapshot.P01.Q01}}
 }
 Speech_C_29 {
  id=#2
  class=Speech_C
  outputs [#15]
  properties {nodeName="", generalConfig {defaultMessage=Hero. Welcome., defaultMetadataParameters [Oracle Convo Start {value=""}]}}
  sub_nodes [
   FortConversationRequirement_ControllerRequirement_38 {
    class=FortConversationRequirement_ControllerRequirement
    properties {
     nodeName=Require: Is Quest Active
     requirement {
      class=FortControllerRequirement_IsQuestActive
      properties {
       bInverseCondition=true
       questDefinition=/BattlePassS29_Quests/Items/QuestItems/Snapshots/Part03/quest_S29_snapshot_p03_q02
       testPolicy=AnyOf
       questDefinitions [
        &quest_S29_snapshot_p01_q01
        &quest_S29_snapshot_p01_q04
        &quest_S29_snapshot_p01_q14
        &quest_S29_snapshot_p03_q08
        &quest_S29_snapshot_p04_q05
        &quest_S29_snapshot_p04_q09
       ]
      }
     }
    }
   }
   ConversationChoiceNode_23 {class=ConversationChoiceNode, properties {defaultChoiceDisplayText=Continue}}
   SideEffect_Client_PlayNPCSoundCueEvent_C_4 {
    class=SideEffect_Client_PlayNPCSoundCueEvent_C
    properties {soundCueToPlay=&VO_ORACLE_S29_2023_00_100_Dialogue_Cue}
   }
   FortConversationRequirement_ControllerRequirement_1 {
    class=FortConversationRequirement_ControllerRequirement
    properties {
     nodeName=Require: Has GameplayTag
     requirement {class=FortControllerRequirement_HasOwnedTag, properties {bInverseCondition=true, ownedTag {tagName=Granted.Quests.Ability.VOPlaying}}}
    }
   }
  ]
 }
 Speech_C_24 {
  id=#15
  class=Speech_C
  outputs [
   #16
   #17
   #18
   #19
   #20
   #21
   #22
   #23
  ]
  properties {generalConfig {
    defaultMessage=Who would you like to contact?
    contextualMessages [{
      message=I am still scrying your fate. Check back later.
      requirementMatchPolicy=RequireAny
      contextRequirements [{requirement {
         class=FortControllerRequirement_IsQuestActive
         properties {
          bInverseCondition=true
          testPolicy=AnyOf
          questDefinitions [
           /BattlePassS29_Quests/Items/QuestItems/Snapshots/Part01/quest_S29_snapshot_p01_q06
           /BattlePassS29_Quests/Items/QuestItems/Snapshots/Part01/quest_S29_snapshot_p01_q10
           /BattlePassS29_Quests/Items/QuestItems/SpringActivation/Drop0/Quest_S29_SA_D00_q01b.quest_S29_SA_D00_q01b
           /BattlePassS29_Quests/Items/QuestItems/SpringActivation/Drop1/quest_S29_SA_D01_q09
           /BattlePassS29_Quests/Items/QuestItems/Snapshots/Part02/quest_S29_snapshot_p02_q01
           /BattlePassS29_Quests/Items/QuestItems/Snapshots/Part03/quest_S29_snapshot_p03_q01
           /BattlePassS29_Quests/Items/QuestItems/Snapshots/Part03/quest_S29_snapshot_p03_q04
           /BattlePassS29_Quests/Items/QuestItems/Snapshots/Part04/quest_S29_snapshot_p04_q01
           /BattlePassS29_Quests/Items/QuestItems/Snapshots/Part04/quest_S29_snapshot_p04_q06
          ]
         }
        }}]
      priority {curve {curveTable=null, rowName=None}}
     }]
    defaultMetadataParameters [OracleEmote {value=Think}]
   }}
  sub_nodes [ConversationChoiceNode_15 {class=ConversationChoiceNode, properties {defaultChoiceDisplayText=Continue}}]
 }
 Speech_C_0 {
  id=#23
  class=Speech_C
  outputs [#24]
  properties {
   softTaskIcon=/BRCosmetics/UI/Foundation/Textures/Icons/Heroes/Athena/Soldier/S29/T-Soldier-Hero-LiftingAuraWave
   generalConfig {defaultMessage=Aphrodite calls...}
  }
  sub_nodes [
   FortConversationChoiceWithPreviewNode_4 {class=FortConversationChoiceWithPreviewNode, properties {defaultChoiceDisplayText=Aphrodite}}
   FortConversationRequirement_ControllerRequirement_0 {
    class=FortConversationRequirement_ControllerRequirement
    properties {
     nodeName=Require: Is Quest Active
     requirement {
      class=FortControllerRequirement_IsQuestActive
      properties {testPolicy=AnyOf, questDefinitions [&quest_S29_snapshot_p02_q01]}
     }
    }
   }
   SideEffect_Client_PlayNPCSoundCueEvent_C_0 {
    class=SideEffect_Client_PlayNPCSoundCueEvent_C
    properties {soundCueToPlay=/BattlePassS29_Quests/Audio/VO/Cues/VO_ORACLE_S29_2024_03_01_010_Dialogue_Cue}
   }
  ]
 }
 ConversationLinkNode_0 {
  id=#24
  class=ConversationLinkNode
  properties {remoteEntryTag {tagName=Conversation.Entry.Season29.Snapshot.P02.Q01}}
 }
 Speech_C_2 {
  id=#22
  class=Speech_C
  outputs [#25]
  properties {
   softTaskIcon=/BRCosmetics/UI/Foundation/Textures/Icons/Heroes/Athena/Soldier/S29/T-Soldier-Hero-GoldAccomplishment
   generalConfig {defaultMessage=Ares is calling...}
  }
  sub_nodes [
   FortConversationChoiceWithPreviewNode_1 {class=FortConversationChoiceWithPreviewNode, properties {defaultChoiceDisplayText=Ares}}
   FortConversationRequirement_ControllerRequirement_4 {
    class=FortConversationRequirement_ControllerRequirement
    properties {
     nodeName=Require: Is Quest Active
     requirement {
      class=FortControllerRequirement_IsQuestActive
      properties {testPolicy=AnyOf, questDefinitions [&quest_S29_snapshot_p01_q10]}
     }
    }
   }
   SideEffect_Client_PlayNPCSoundCueEvent_C_10 {
    class=SideEffect_Client_PlayNPCSoundCueEvent_C
    properties {soundCueToPlay=/BattlePassS29_Quests/Audio/VO/Cues/VO_ORACLE_S29_2024_02_01_010_Dialogue_Cue}
   }
  ]
 }
 ConversationLinkNode_11 {
  id=#25
  class=ConversationLinkNode
  properties {remoteEntryTag {tagName=Conversation.Entry.Season29.Snapshot.P01.Q10}}
 }
 Speech_C_5 {
  id=#21
  class=Speech_C
  outputs [#26]
  properties {
   softTaskIcon=/BRCosmetics/UI/Foundation/Textures/Icons/Heroes/Athena/Soldier/S29/T-Soldier-Hero-ReptilianOceanSleek
   generalConfig {defaultMessage=Medusa beckons...}
  }
  sub_nodes [
   FortConversationChoiceWithPreviewNode_2 {class=FortConversationChoiceWithPreviewNode, properties {defaultChoiceDisplayText=Medusa}}
   FortConversationRequirement_ControllerRequirement_5 {
    class=FortConversationRequirement_ControllerRequirement
    properties {
     nodeName=Require: Is Quest Active
     requirement {
      class=FortControllerRequirement_IsQuestActive
      properties {testPolicy=AnyOf, questDefinitions [/BattlePassS29_Quests/Items/QuestItems/Snapshots/Part04/quest_S29_snapshot_p04_q07]}
     }
    }
   }
   SideEffect_Client_PlayNPCSoundCueEvent_C_11 {
    class=SideEffect_Client_PlayNPCSoundCueEvent_C
    properties {soundCueToPlay=/BattlePassS29_Quests/Audio/VO/Cues/VO_ORACLE_S29_2024_05_01_010_Dialogue_Cue}
   }
  ]
 }
 ConversationLinkNode_14 {
  id=#26
  class=ConversationLinkNode
  properties {remoteEntryTag {tagName=Conversation.Entry.Season29.Snapshot.P04.Q07}}
 }
 Speech_C_6 {
  id=#20
  class=Speech_C
  outputs [#27]
  properties {
   softTaskIcon=&T-Soldier-Hero-ReptilianOceanSleek
   generalConfig {defaultMessage=Medusa beckons...}
  }
  sub_nodes [
   FortConversationChoiceWithPreviewNode_3 {class=FortConversationChoiceWithPreviewNode, properties {defaultChoiceDisplayText=Medusa}}
   FortConversationRequirement_ControllerRequirement_6 {
    class=FortConversationRequirement_ControllerRequirement
    properties {
     nodeName=Require: Is Quest Active
     requirement {
      class=FortControllerRequirement_IsQuestActive
      properties {testPolicy=AnyOf, questDefinitions [&quest_S29_snapshot_p04_q01]}
     }
    }
   }
   SideEffect_Client_PlayNPCSoundCueEvent_C_12 {
    class=SideEffect_Client_PlayNPCSoundCueEvent_C
    properties {soundCueToPlay=&VO_ORACLE_S29_2024_05_01_010_Dialogue_Cue}
   }
  ]
 }
 ConversationLinkNode_15 {
  id=#27
  class=ConversationLinkNode
  properties {remoteEntryTag {tagName=Conversation.Entry.Season29.Snapshot.P04.Q01}}
 }
 Speech_C_18 {
  id=#19
  class=Speech_C
  outputs [#28, #29]
  properties {
   softTaskIcon=/BRCosmetics/UI/Foundation/Textures/Icons/Heroes/Athena/Soldier/S29/T-Soldier-Hero-DarkStanceInferno
   generalConfig {defaultMessage=Hades beckons...}
  }
  sub_nodes [
   FortConversationChoiceWithPreviewNode_10 {class=FortConversationChoiceWithPreviewNode, properties {defaultChoiceDisplayText=Hades}}
   FortConversationRequirement_ControllerRequirement_15 {
    class=FortConversationRequirement_ControllerRequirement
    properties {
     nodeName=Require: Is Quest Active
     requirement {
      class=FortControllerRequirement_IsQuestActive
      properties {
       testPolicy=AnyOf
       questDefinitions [
        &quest_S29_snapshot_p01_q06
        &quest_S29_snapshot_p03_q04
       ]
      }
     }
    }
   }
   SideEffect_Client_PlayNPCSoundCueEvent_C_18 {
    class=SideEffect_Client_PlayNPCSoundCueEvent_C
    properties {soundCueToPlay=/BattlePassS29_Quests/Audio/VO/Cues/VO_ORACLE_S29_2024_02_02_010_Dialogue_Cue}
   }
  ]
 }
 ConversationLinkNode_8 {
  id=#29
  class=ConversationLinkNode
  properties {remoteEntryTag {tagName=Conversation.Entry.Season29.Snapshot.P01.Q06}}
 }
 ConversationLinkNode_1 {
  id=#28
  class=ConversationLinkNode
  properties {remoteEntryTag {tagName=Conversation.Entry.Season29.Snapshot.P03.Q04}}
 }
 Speech_C_9 {
  id=#18
  class=Speech_C
  outputs [#30]
  properties {
   softTaskIcon=/BRCosmetics/UI/Foundation/Textures/Icons/Heroes/Athena/Soldier/S29/T-Soldier-Hero-GateHoundSpectral
   generalConfig {defaultMessage=Cerberus is calling.}
  }
  sub_nodes [
   FortConversationChoiceWithPreviewNode_6 {class=FortConversationChoiceWithPreviewNode, properties {defaultChoiceDisplayText=Cerberus}}
   FortConversationRequirement_ControllerRequirement_9 {
    class=FortConversationRequirement_ControllerRequirement
    properties {
     nodeName=Require: Is Quest Active
     requirement {
      class=FortControllerRequirement_IsQuestActive
      properties {testPolicy=AnyOf, questDefinitions [&quest_S29_snapshot_p03_q01]}
     }
    }
   }
   SideEffect_Client_PlayNPCSoundCueEvent_C_14 {
    class=SideEffect_Client_PlayNPCSoundCueEvent_C
    properties {soundCueToPlay=/BattlePassS29_Quests/Audio/VO/Cues/VO_ORACLE_S29_2024_04_01_010_Dialogue_Cue}
   }
  ]
 }
 ConversationLinkNode_20 {
  id=#30
  class=ConversationLinkNode
  properties {remoteEntryTag {tagName=Conversation.Entry.Season29.Snapshot.P03.Q01}}
 }
 Speech_C_11 {
  id=#17
  class=Speech_C
  outputs [#31]
  properties {
   softTaskIcon=/BRCosmetics/UI/Foundation/Textures/Icons/Heroes/Athena/Soldier/T-Soldier-HID-692-Athena-Commando-M-HenchmanTough
   generalConfig {defaultMessage="Is that you, Brutus?"}
  }
  sub_nodes [
   FortConversationChoiceWithPreviewNode_7 {class=FortConversationChoiceWithPreviewNode, properties {defaultChoiceDisplayText=Brutus}}
   FortConversationRequirement_ControllerRequirement_10 {
    class=FortConversationRequirement_ControllerRequirement
    properties {
     nodeName=Require: Is Quest Active
     requirement {
      class=FortControllerRequirement_IsQuestActive
      properties {testPolicy=AnyOf, questDefinitions [&quest_S29_SA_D01_q09]}
     }
    }
   }
   SideEffect_Client_PlayNPCSoundCueEvent_C_15 {
    class=SideEffect_Client_PlayNPCSoundCueEvent_C
    properties {soundCueToPlay=/BattlePassS29_Quests/Audio/VO/Cues/VO_ORACLE_S29_2023_00_210_Dialogue_Cue}
   }
  ]
 }
 ConversationLinkNode_21 {
  id=#31
  class=ConversationLinkNode
  properties {remoteEntryTag {tagName=Conversation.Entry.Season29.SeasonalActivation.09}}
 }
 Speech_C_14 {
  id=#16
  class=Speech_C
  outputs [#32]
  properties {
   softTaskIcon=/BRCosmetics/UI/Foundation/Textures/Icons/Heroes/Athena/Soldier/T-Soldier-HID-A-034-Athena-Commando-F-CatBurglar
   generalConfig {defaultMessage=Marigold speaks.}
  }
  sub_nodes [
   FortConversationChoiceWithPreviewNode_8 {class=FortConversationChoiceWithPreviewNode, properties {defaultChoiceDisplayText=Marigold}}
   FortConversationRequirement_ControllerRequirement_11 {
    class=FortConversationRequirement_ControllerRequirement
    properties {
     nodeName=Require: Is Quest Active
     requirement {
      class=FortControllerRequirement_IsQuestActive
      properties {
       testPolicy=AnyOf
       questDefinitions [&quest_S29_SA_D00_q01b]
      }
     }
    }
   }
   SideEffect_Client_PlayNPCSoundCueEvent_C_16 {
    class=SideEffect_Client_PlayNPCSoundCueEvent_C
    properties {soundCueToPlay=/BattlePassS29_Quests/Audio/VO/Cues/VO_ORACLE_S29_2023_00_190_Dialogue_Cue}
   }
  ]
 }
 ConversationLinkNode_22 {
  id=#32
  class=ConversationLinkNode
  properties {remoteEntryTag {tagName=Conversation.Entry.Season29.SeasonalActivation.01b}}
 }
]
